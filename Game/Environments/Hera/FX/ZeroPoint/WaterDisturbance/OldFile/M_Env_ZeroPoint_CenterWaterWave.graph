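# Translucent lit water material, in an OldFile folder, for the central wave of the Zero Point water disturbance effect.
# A custom HLSL node sums six Gerstner waves into offset, normal and per-wave masks, feeding edge-masked vertex displacement along the mesh normal.
# Height gradients tint the water, while panned foam glows on wave fronts and height-based spikes.
# Height, depth and camera fades, view-angle-limited refraction and Fresnel sky-light reflection shape the surface.
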Material /Game/Environments/Hera/FX/ZeroPoint/WaterDisturbance/OldFile/M_Env_ZeroPoint_CenterWaterWave
graph {
 properties {
  bAutomaticallySetUsageInEditor=false
  blendMode=BLEND_Translucent
  refractionMethod=RM_IndexOfRefraction
  translucencyLightingMode=TLM_VolumetricDirectional
  usedShadingModels=Default Lit
 }
 nodes [
  n1 = ScalarParameter() {defaultValue=8, parameterName=FresE}
  n2 = ScalarParameter() {parameterName=FresBaseReflectFractionIn}
  n3 = Transform(None=NamedRerouteUsage()) {transformSourceType=TRANSFORMSOURCE_Local}
  n4 = Transform(None=MaterialFunctionCall(
    Normal=Transform(None=@n3) {transformSourceType=TRANSFORMSOURCE_World, transformType=TRANSFORM_Tangent}
    Flatness=ScalarParameter() {defaultValue=0.6305, parameterName=DLM}
   ) {materialFunction=&FlattenNormal}.Result)
  n5 = SkyLightEnvMapSample(Direction=ReflectionVectorWS(CustomWorldNormal=@n4), Roughness=ScalarParameter() {parameterName=R2})
  n6 = VectorParameter() {parameterName=Ref, defaultValue=<rgba 0 0 0 1>}
  Add(A=Multiply(A=Fresnel(ExponentIn=@n1, BaseReflectFractionIn=@n2, Normal=@n4), B=Multiply(A=@n5, B=@n6.RGB)))
  Constant()
  n7 = MaterialFunctionCall() {materialFunction=&CameraDirectionVector}
  n8 = Normalize(VectorInput=Constant3Vector() {constant=<rgba 1 1 0 0>})
  n9 = Saturate(None=DotProduct(A=@n7.Result, B=Transform(None=@n8)))
  n10 = ScalarParameter() {defaultValue=0.5, parameterName=XYAngleLimit_Offset}
  n11 = Saturate(None=Multiply(A=Subtract(A=@n9, B=@n10), B=ScalarParameter() {defaultValue=4, parameterName=XYAngleLimit_Mul}))
  n12 = Abs(None=DotProduct(A=@n7.Result, B=Constant3Vector() {constant=<rgba 0 0 1 0>}))
  n13 = ScalarParameter() {defaultValue=0.224, parameterName=ZAngleLimit_Offset}
  n14 = Saturate(None=Multiply(A=Subtract(A=@n12, B=@n13), B=ScalarParameter() {defaultValue=10, parameterName=ZAngleLimit_Mul}))
  n15 = OneMinus(None=Saturate(None=LinearInterpolate(A=@n11, Alpha=@n14) {constB=0}))
  n16 = ScalarParameter() {defaultValue=230, parameterName=t}
  n17 = Divide(A=@n16) {constB=-2}
  n18 = NamedRerouteUsage()
  MaterialExpressionTextureCoordinate_1 = TextureCoordinate()
  n19 = Convert(None=@MaterialExpressionTextureCoordinate_1) {
   nodeName=Break Float2
   convertInputs [{
     type=Vector2
     expressionInput {
      expression=&self:MaterialExpressionTextureCoordinate_1
      inputName=None
      mask=0
      maskA=0
      maskB=0
      maskG=0
      maskR=0
      outputIndex=0
     }
    }]
   convertOutputs [{type=Scalar}, {type=Scalar}]
  }
  n20 = VectorParameter() {
   group=WPOEdgeMask
   parameterName=WPOEdgeMask_EdgeMask_InvertControl
   channelNames {a="", b=Power, g=Multiply, r=Offset}
   defaultValue=<rgba 0 4 1 1>
  }
  n21 = MaterialFunctionCall!(UorV=@n19, Invert_EdgeControl=@n20.RGB, NoInvert_EdgeControl=@n20.RGB) {materialFunction=/Game/Effects/Niagara/MasterMaterials/Functions/Common/FXCommon_EdgeMask_Ex}
  n22 = VectorParameter() {
   group=WPOEdgeMask
   parameterName=WPOEdgeMask_EdgeMask_Control
   channelNames {a="", b=Power, g=Multiply, r=Offset}
   defaultValue=<rgba 0 4 1 1>
  }
  n23 = MaterialFunctionCall(UorV=@n19, Invert_EdgeControl=@n20.RGB, NoInvert_EdgeControl=@n22.RGB) {materialFunction=&FXCommon_EdgeMask_Ex}
  n24 = Multiply(A=LinearInterpolate(A=@n17, B=Multiply(A=@n18, B=@n16), Alpha=Multiply(A=@n21.Mask, B=@n23.Mask)), B=VertexNormalWS())
  MaterialExpressionAdd_0 = Add(A=@n24, B=WorldPosition().XYZ)
  n25 = Convert(None=@MaterialExpressionAdd_0) {
   nodeName=Break Float3
   convertInputs [{
     type=Vector3
     expressionInput {
      expression=&self:MaterialExpressionAdd_0
      inputName=None
      mask=0
      maskA=0
      maskB=0
      maskG=0
      maskR=0
      outputIndex=0
     }
    }]
   convertOutputs [{type=Scalar}, {type=Scalar}, {type=Scalar}]
  }
  n26 = ScalarParameter() {defaultValue=1380, parameterName=WaterSurfaceHeight}
  n27 = Divide(A=Max(A=Subtract(A=@n25, B=@n26)) {constB=0}, B=ScalarParameter() {defaultValue=1083.0216, parameterName=WaterSurfaceFade})
  n28 = ScalarParameter() {defaultValue=1.2, parameterName=SC2}
  n29 = Power(Base=OneMinus(None=Power(Base=@n27, Exp=@n28)), Exp=ScalarParameter() {defaultValue=4, parameterName=SC})
  n30 = ScalarParameter() {defaultValue=100, parameterName=DeF}
  n31 = Power(Base=MaterialFunctionCall(Distance=ScalarParameter() {defaultValue=9333.1758, parameterName=StartFadeDistance}) {materialFunction=&Near_Camera_Fade}.Result) {constExponent=8}
  n32 = Multiply(
   A=Multiply(A=OneMinus(None=@n29), B=DepthFade(FadeDistance=@n30))
   B=LinearInterpolate(
    A=ScalarParameter() {defaultValue=0.5, parameterName=SCV_OpacityScale}
    B=ScalarParameter() {defaultValue=0.25, parameterName=SCV_OpacityScale2}
    Alpha=@n31
   )
  )
  n33 = Power(Base=MaterialFunctionCall(Distance=ScalarParameter() {defaultValue=663.9774, parameterName=NCF_Distance}) {materialFunction=&Near_Camera_Fade}.Result) {constExponent=3}
  LinearInterpolate(B=ScalarParameter() {defaultValue=1.1, parameterName=Ref}, Alpha=Multiply(A=Multiply(A=@n15, B=@n32), B=@n33)) {constA=1}
  MaterialExpressionRotator_0 = Rotator(Coordinate=TextureCoordinate(), Time=MaterialFunctionCall() {materialFunction=&Pi}.Result)
  n34 = Convert(None=@MaterialExpressionRotator_0) {
   nodeName=Break Float2
   convertInputs [{
     type=Vector2
     expressionInput {
      expression=&self:MaterialExpressionRotator_0
      inputName=None
      mask=0
      maskA=0
      maskB=0
      maskG=0
      maskR=0
      outputIndex=0
     }
    }]
   convertOutputs [{type=Scalar}, {type=Scalar}]
  }
  n35 = VectorParameter() {
   group=StrenghtMask
   parameterName=StrenghtMask_EdgeMask_InvertControl
   channelNames {a="", b=Power, g=Multiply, r=Offset}
   defaultValue=<rgba 0 4.5 0.5 1>
  }
  n36 = VectorParameter() {
   group=StrenghtMask
   parameterName=StrenghtMask_EdgeMask_Control
   channelNames {a="", b=Power, g=Multiply, r=Offset}
   defaultValue=<rgba 10 1 1 1>
  }
  LinearInterpolate(
   A=@n17
   B=Multiply(A=@n24, B=@n33)
   Alpha=MaterialFunctionCall(UorV=@n34, Invert_EdgeControl=@n35.RGB, NoInvert_EdgeControl=@n36.RGB) {materialFunction=&FXCommon_EdgeMask_Ex}.Mask
  )
  MaterialExpressionMultiply_16 = Multiply(A=TextureCoordinate(), B=ScalarParameter() {defaultValue=1437.7084, parameterName=N})
  MaterialExpressionObjectPositionWS_0 = ObjectPositionWS()
  n37 = Convert(None=@MaterialExpressionObjectPositionWS_0) {
   nodeName=Break Float3
   convertInputs [{
     type=Vector3
     expressionInput {
      expression=&self:MaterialExpressionObjectPositionWS_0
      inputName=None
      mask=0
      maskA=0
      maskB=0
      maskG=0
      maskR=0
      outputIndex=0
     }
    }]
   convertOutputs [{type=Scalar}, {type=Scalar}, {type=Scalar}]
  }
  n38 = ScalarParameter() {parameterName=TR}
  MaterialExpressionMultiply_17 = Multiply(A=Add(A=Time() {bOverride_Period=true, period=6000}, B=Multiply(A=@n37, B=@n38)), B=ScalarParameter() {defaultValue=1.5, parameterName=TS})
  MaterialExpressionScalarParameter_201 = ScalarParameter() {defaultValue=1.1, parameterName=Steepness}
  MaterialExpressionScalarParameter_202 = ScalarParameter() {defaultValue=0.5, parameterName=Amplitude}
  MaterialExpressionScalarParameter_8 = ScalarParameter() {defaultValue=50, parameterName=NormalStrength}
  n39 = Custom!(
   Position=@MaterialExpressionMultiply_16
   Time=@MaterialExpressionMultiply_17
   Steepness=@MaterialExpressionScalarParameter_201
   Amplitude=@MaterialExpressionScalarParameter_202
   NormalStrength=@MaterialExpressionScalarParameter_8
  ) {
   code="float2 dirs[6];\r\ndirs[0] = normalize(float2( 1.00,  0.20));\r\ndirs[1] = normalize(float2( 0.45,  0.89));\r\ndirs[2] = normalize(float2(-0.60,  0.80));\r\ndirs[3] = normalize(float2(-0.92, -0.39));\r\ndirs[4] = normalize(float2( 0.18, -0.98));\r\ndirs[5] = normalize(float2( 0.78, -0.62));\r\nfloat wavelengths[6] = { 1500.0, 900.0, 520.0, 300.0, 170.0, 95.0 };\r\nfloat amps[6]        = {  1.00,  0.62,  0.40,  0.26,  0.16,  0.10 };\r\n\r\nfloat3 offset = float3(0, 0, 0);\r\nfloat Nx = 0.0;\r\nfloat Ny = 0.0;\r\nfloat Nz = 1.0;\r\nfloat waveGray[6];\r\n\r\n[unroll]\r\nfor (int i = 0; i < 6; i++)\r\n{\r\n    float A   = Amplitude * amps[i];\r\n    float w   = 6.28318530718 / wavelengths[i];\r\n    float phi = sqrt(9.81 * w * 100.0);\r\n    float Q   = Steepness / (w * A * 6.0);\r\n    float theta = dot(dirs[i], Position.xy) * w + Time * phi;\r\n    float c = cos(theta);\r\n    float s = sin(theta);\r\n\r\n    offset.x += Q * A * dirs[i].x * c;\r\n    offset.y += Q * A * dirs[i].y * c;\r\n    offset.z += A * s;\r\n\r\n    float WA = w * A;\r\n    Nx -= dirs[i].x * WA * c;\r\n    Ny -= dirs[i].y * WA * c;\r\n    Nz -= Q * WA * s;\r\n\r\n    waveGray[i] = s * 0.5 + 0.5;\r\n}\r\n\r\nNx *= NormalStrength;\r\nNy *= NormalStrength;\r\n\r\nNormal = normalize(float3(Nx, Ny, Nz));\r\n\r\nWave0 = waveGray[0];\r\nWave1 = waveGray[1];\r\nWave2 = waveGray[2];\r\nWave3 = waveGray[3];\r\nWave4 = waveGray[4];\r\nWave5 = waveGray[5];\r\n\r\nreturn offset;"
   showCode=true
   additionalOutputs [
    {outputName=Normal, outputType=CMOT_Float3}
    {outputName=Wave0, outputType=CMOT_Float1}
    {outputName=Wave1, outputType=CMOT_Float1}
    {outputName=Wave2, outputType=CMOT_Float1}
    {outputName=Wave3, outputType=CMOT_Float1}
    {outputName=Wave4, outputType=CMOT_Float1}
    {outputName=Wave5, outputType=CMOT_Float1}
   ]
   inputs [
    {
     inputName=Position
     input {expression=&self:MaterialExpressionMultiply_16}
    }
    {
     inputName=Time
     input {expression=&self:MaterialExpressionMultiply_17}
    }
    {
     inputName=Steepness
     input {expression=&self:MaterialExpressionScalarParameter_201}
    }
    {
     inputName=Amplitude
     input {expression=&self:MaterialExpressionScalarParameter_202}
    }
    {
     inputName=NormalStrength
     input {expression=&self:MaterialExpressionScalarParameter_8}
    }
   ]
  }
  NamedRerouteDeclaration(None=VertexInterpolator(VS=@n39.Normal).PS) {name=NormalOut, nodeColor=<rgba 0 1 1 1>}
  NamedRerouteDeclaration(None=@n39.return) {name=WPOOffset, nodeColor=<rgba 1 0.6797 0 1>}
  OneMinus(None=@n31)
  n40 = VectorParameter() {parameterName=WaveDirection, defaultValue=<rgba 1 1 0 1>}
  n41 = Multiply(A=DotProduct(A=Normalize(VectorInput=@n40.RGB), B=NamedRerouteUsage())) {constB=-1}
  n42 = ScalarParameter() {defaultValue=-0.04, group=WaveFrontMask, parameterName=WaveFrontMask_GrayAdjust_Add}
  n43 = Saturate(None=Multiply(A=Add(A=@n41, B=@n42), B=ScalarParameter() {defaultValue=1, group=WaveFrontMask, parameterName=WaveFrontMask_GrayAdjust_Mul}))
  n44 = ScalarParameter() {defaultValue=3, group=WaveFrontMask, parameterName=WaveFrontMask_GrayAdjust_Power}
  n45 = Multiply(A=VertexInterpolator(VS=@n39.Wave4).PS, B=ScalarParameter() {defaultValue=110.8251, parameterName=Offsetddd})
  n46 = Add(A=@n45, B=ScalarParameter() {defaultValue=-1879.287, group=Spike, parameterName=Spike_GrayAdjust_Add})
  n47 = Saturate(None=Multiply(A=Add(A=@n25, B=@n46), B=ScalarParameter() {defaultValue=0.0005, group=Spike, parameterName=Spike_GrayAdjust_Mul}))
  n48 = ScalarParameter() {defaultValue=1, group=Spike, parameterName=Spike_GrayAdjust_Power}
  n49 = Add(A=Saturate(None=Power(Base=@n43, Exp=@n44)), B=Saturate(None=Power(Base=@n47, Exp=@n48)))
  n50 = VectorParameter() {group=FoamT, parameterName=FoamT_UVControl, channelNames {a=VPan, b=UPan, g=VScale, r=UScale}, defaultValue=<rgba 1.25 1.25 -0.1 -0.1>}
  n51 = MaterialFunctionCall(
   Texture=TextureObject() {samplerType=SAMPLERTYPE_Masks, texture=/FortWater/Textures/T_OceanFoam_New_Mask}
   UVControl_RGB=@n50.RGB
   UVControl_A=@n50.RGB
  ) {materialFunction=&FXCommon_TextureSample}
  n52 = ScalarParameter() {defaultValue=2, parameterName=a2}
  n53 = Multiply(
   A=VectorParameter() {parameterName=BC, defaultValue=<rgba 23.4323 7.8261 60 1>}.RGB
   B=Multiply(A=Saturate(None=@n49), B=Power(Base=@n51.R, Exp=@n52))
  )
  Reroute(None=MaterialFunctionCall(Emissive=@n53) {materialFunction=&MF_FX_EmissiveAdjustments}.Result)
  n54 = VectorParameter() {parameterName=AColor, defaultValue=<rgba 0.04 0.2779 0.4 1>}
  n55 = Subtract(A=Add(A=@n45, B=@n25), B=ScalarParameter() {defaultValue=1622.3865, parameterName=WaterSurfaceHeightAF})
  n56 = OneMinus(None=Divide(A=Max(A=@n55) {constB=0}, B=ScalarParameter() {defaultValue=240.3914, parameterName=WaterColor}))
  n57 = ScalarParameter() {defaultValue=0.3704, group=WaterColor, parameterName=WaterColor_G2C_InvtPower}
  n58 = LinearInterpolate(
   A=VectorParameter() {group=WaterColor, parameterName=OPV_WaterColor_G2C_ColorA, defaultValue=<rgba 0.0741 0.337 0.6 0>}.RGB
   B=VectorParameter() {group=WaterColor, parameterName=OPV_WaterColor_G2C_ColorB, defaultValue=<rgba 0 0.7 0.5367 0>}.RGB
   Alpha=OneMinus(None=Power(Base=@n56, Exp=@n57))
  )
  n59 = Multiply(
   A=Add(A=NamedRerouteUsage(), B=ScalarParameter() {defaultValue=30.9927, group=ColorG, parameterName=ColorG_GrayAdjust_Add})
   B=ScalarParameter() {defaultValue=0.0576, group=ColorG, parameterName=ColorG_GrayAdjust_Mul}
  )
  n60 = Saturate(None=Power(Base=Saturate(None=@n59), Exp=ScalarParameter() {defaultValue=2, group=ColorG, parameterName=ColorG_GrayAdjust_Power}))
  Reroute(None=Add(A=LinearInterpolate(A=@n54.RGB, B=@n58, Alpha=@n60), B=@n53))
  Reroute(None=@n3)
 ]
}
