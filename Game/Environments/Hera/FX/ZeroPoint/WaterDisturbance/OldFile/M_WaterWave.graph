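# Translucent lit water-wave material for Niagara mesh particles, in the Zero Point water disturbance FX folder.
# A custom HLSL node builds scrolling, overlapping rows of rounded bumps, returning height and analytically derived normals.
# Edge-masked height feeds a two-color blend, depth-faded opacity with particle alpha, and a noise-modulated displacement along vertex normals.
# The normals drive a Fresnel-weighted skylight reflection for emissive, while dynamic parameters control displacement strength, peak position and scroll offset.
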Material /Game/Environments/Hera/FX/ZeroPoint/WaterDisturbance/OldFile/M_WaterWave
graph {
 expressions {
  MaterialExpressionAdd_0 {
   class=MaterialExpressionAdd
   inputs [
    A {from=MaterialExpressionTextureCoordinate_0}
    B {from=MaterialExpressionConvert_0}
   ]
  }
  MaterialExpressionAdd_1 {
   class=MaterialExpressionAdd
   inputs [
    A {from=MaterialExpressionMultiply_9}
    B {from=MaterialExpressionScalarParameter_18}
   ]
  }
  MaterialExpressionAdd_2 {
   class=MaterialExpressionAdd
   inputs [
    A {from=MaterialExpressionMultiply_9}
    B {from=MaterialExpressionScalarParameter_6}
   ]
  }
  MaterialExpressionAdd_3 {
   class=MaterialExpressionAdd
   inputs [
    A {from=MaterialExpressionTime_0}
    B {from=MaterialExpressionDynamicParameter_1, output=A}
   ]
  }
  MaterialExpressionAppendVector_0 {
   class=MaterialExpressionAppendVector
   inputs [
    A {from=MaterialExpressionVectorParameter_0, output=R}
    B {from=MaterialExpressionVectorParameter_0, output=R}
   ]
  }
  MaterialExpressionComponentMask_0 {class=MaterialExpressionComponentMask, inputs [None {from=MaterialExpressionCustom_0}], properties {b=true, g=true, r=true}}
  MaterialExpressionComponentMask_1 {class=MaterialExpressionComponentMask, inputs [None {from=MaterialExpressionCustom_0}], properties {a=true}}
  MaterialExpressionConstant_0 {class=MaterialExpressionConstant}
  MaterialExpressionConstant_1 {class=MaterialExpressionConstant, properties {r=1}}
  MaterialExpressionConvert_0 {
   class=MaterialExpressionConvert
   inputs [
    None {from=""}
    None {from=MaterialExpressionMultiply_6}
   ]
   properties {
    nodeName=Make Vector2
    convertInputs [
     {type=Scalar, expressionInput {expression=null, inputName=None, mask=0, maskA=0, maskB=0, maskG=0, maskR=0, outputIndex=0}}
     {
      type=Scalar
      expressionInput {
       expression=&self:MaterialExpressionMultiply_6
       inputName=None
       mask=0
       maskA=0
       maskB=0
       maskG=0
       maskR=0
       outputIndex=0
      }
     }
    ]
    convertOutputs [{type=Vector2}]
   }
  }
  MaterialExpressionConvert_1 {
   class=MaterialExpressionConvert
   inputs [None {from=MaterialExpressionTextureCoordinate_1}]
   properties {
    nodeName=Break Float2
    convertInputs [{
      type=Vector2
      expressionInput {
       expression=&self:MaterialExpressionTextureCoordinate_1
       inputName=None
       mask=0
       maskA=0
       maskB=0
       maskG=0
       maskR=0
       outputIndex=0
      }
     }]
    convertOutputs [{type=Scalar}, {type=Scalar}]
   }
  }
  MaterialExpressionCustom_0 {
   class=MaterialExpressionCustom
   inputs [
    UV {from=MaterialExpressionAdd_0, output=""}
    Tiling {from=MaterialExpressionAppendVector_0, output=""}
    Falloff {from=MaterialExpressionVectorParameter_0, output=B}
    NormalStrength {from=MaterialExpressionVectorParameter_0, output=B}
    Roundness {from=MaterialExpressionScalarParameter_22, output=""}
    CenterOffsetV {from=MaterialExpressionDynamicParameter_1, output=G}
    RowOverlap {from=MaterialExpressionScalarParameter_11, output=""}
   ]
   properties {
    code="float2 uv = UV * Tiling;\r\n\r\nfloat fx = frac(uv.x);\r\nfloat cx = fx - 0.5;\r\nfloat qx = abs(2.0 * cx);\r\n\r\nfloat peakV      = clamp(0.5 + CenterOffsetV, 0.001, 0.999);\r\nfloat halfBelow  = peakV         * RowOverlap;\r\nfloat halfAbove  = (1.0 - peakV) * RowOverlap;\r\nfloat nearestRow = floor(uv.y - peakV + 0.5);\r\nfloat distSelf   = uv.y - peakV - nearestRow;\r\n\r\nfloat p = max(Roundness, 1.0);\r\n\r\nfloat sgnS = (distSelf < 0.0) ? -1.0 : 1.0;\r\nfloat invS = (distSelf < 0.0) ? 1.0 / halfBelow : 1.0 / halfAbove;\r\nfloat qy_S = saturate(abs(distSelf) * invS);\r\nfloat B_S  = 1.0 - pow(qy_S, p);\r\n\r\nfloat qy_D = saturate((1.0 + distSelf) / halfAbove);\r\nfloat B_D  = 1.0 - pow(qy_D, p);\r\n\r\nfloat qy_U = saturate((1.0 - distSelf) / halfBelow);\r\nfloat B_U  = 1.0 - pow(qy_U, p);\r\n\r\nfloat B = B_S, qy = qy_S, sgn = sgnS, inv = invS;\r\nif (B_D > B) { B = B_D; qy = qy_D; sgn =  1.0; inv = 1.0 / halfAbove; }\r\nif (B_U > B) { B = B_U; qy = qy_U; sgn = -1.0; inv = 1.0 / halfBelow; }\r\n\r\nfloat A = 1.0 - pow(qx, p);\r\nfloat d = saturate(A * B);\r\nfloat h = pow(d, Falloff);\r\n\r\nfloat qxPow = pow(max(qx, 1e-5), p - 1.0);\r\nfloat qyPow = pow(max(qy, 1e-5), p - 1.0);\r\nfloat dAdu  = -2.0 * p * sign(cx) * qxPow;\r\nfloat dBdv  = -p * qyPow * sgn * inv;\r\n\r\nfloat dPow = Falloff * pow(max(d, 1e-5), Falloff - 1.0);\r\nfloat dhdu = dPow * dAdu * B;\r\nfloat dhdv = dPow * A    * dBdv;\r\n\r\nfloat3 N = normalize(float3(-dhdu * NormalStrength,\r\n                            -dhdv * NormalStrength,\r\n                             1.0));\r\nreturn float4(N, h);"
    outputType=CMOT_Float4
    showCode=true
    inputs [
     {
      inputName=UV
      input {
       expression=&self:MaterialExpressionAdd_0
       inputName=None
       mask=0
       maskA=0
       maskB=0
       maskG=0
       maskR=0
       outputIndex=0
      }
     }
     {
      inputName=Tiling
      input {
       expression=&self:MaterialExpressionAppendVector_0
       inputName=None
       mask=0
       maskA=0
       maskB=0
       maskG=0
       maskR=0
       outputIndex=0
      }
     }
     {
      inputName=Falloff
      input {
       expression=&self:MaterialExpressionVectorParameter_0
       inputName=None
       mask=1
       maskA=0
       maskB=1
       maskG=0
       maskR=0
       outputIndex=3
      }
     }
     {
      inputName=NormalStrength
      input {
       expression=&self:MaterialExpressionVectorParameter_0
       inputName=None
       mask=1
       maskA=1
       maskB=0
       maskG=0
       maskR=0
       outputIndex=4
      }
     }
     {
      inputName=Roundness
      input {
       expression=&self:MaterialExpressionScalarParameter_22
       inputName=None
       mask=0
       maskA=0
       maskB=0
       maskG=0
       maskR=0
       outputIndex=0
      }
     }
     {
      inputName=CenterOffsetV
      input {
       expression=&self:MaterialExpressionDynamicParameter_1
       inputName=None
       mask=1
       maskA=0
       maskB=0
       maskG=1
       maskR=0
       outputIndex=1
      }
     }
     {
      inputName=RowOverlap
      input {
       expression=&self:MaterialExpressionScalarParameter_11
       inputName=None
       mask=0
       maskA=0
       maskB=0
       maskG=0
       maskR=0
       outputIndex=0
      }
     }
    ]
   }
  }
  MaterialExpressionDepthFade_0 {
   class=MaterialExpressionDepthFade
   inputs [
    Opacity {from="", output=""}
    FadeDistance {from=MaterialExpressionScalarParameter_3, output=""}
   ]
   properties {}
  }
  MaterialExpressionDynamicParameter_0 {class=MaterialExpressionDynamicParameter, inputs [], properties {defaultValue=<rgba 1 0 1 1>}}
  MaterialExpressionDynamicParameter_1 {class=MaterialExpressionDynamicParameter, inputs [], properties {defaultValue=<rgba 1 0 1 1>}}
  MaterialExpressionFresnel_0 {
   class=MaterialExpressionFresnel
   inputs [
    ExponentIn {from=MaterialExpressionScalarParameter_4, output=""}
    BaseReflectFractionIn {from=MaterialExpressionScalarParameter_5, output=""}
    Normal {from=MaterialExpressionReroute_0, output=""}
   ]
   properties {}
  }
  MaterialExpressionLinearInterpolate_3 {
   class=MaterialExpressionLinearInterpolate
   inputs [
    A {from=MaterialExpressionVectorParameter_1, output=RGB}
    B {from=MaterialExpressionVectorParameter_9, output=RGB}
    Alpha {from=MaterialExpressionSaturate_5}
   ]
  }
  MaterialExpressionMaterialFunctionCall_1 {
   class=MaterialExpressionMaterialFunctionCall
   inputs [
    Texture {from=MaterialExpressionTextureObjectParameter_0}
    AdditionalUVOffset {from=MaterialExpressionMultiply_11}
    UVControl_RGB {from=MaterialExpressionVectorParameter_3, output=RGB}
    UVControl_A {from=MaterialExpressionVectorParameter_3, output=RGB}
    "[Override]TimeInput" {from=""}
    "[Override]UVInput" {from=""}
   ]
   properties {materialFunction=&FXCommon_TextureSample}
  }
  MaterialExpressionMaterialFunctionCall_2 {
   class=MaterialExpressionMaterialFunctionCall
   inputs [
    Input {from=MaterialExpressionMaterialFunctionCall_1, output=R}
    Input Low {from=MaterialExpressionConstant_0}
    Input High {from=MaterialExpressionConstant_1}
    Target Low {from=MaterialExpressionScalarParameter_12}
    Target High {from=MaterialExpressionConstant_1}
   ]
   properties {materialFunction=&RemapValueRange}
  }
  MaterialExpressionMaterialFunctionCall_3 {
   class=MaterialExpressionMaterialFunctionCall
   inputs [Emissive {from=MaterialExpressionMultiply_4}]
   properties {materialFunction=&MF_FX_EmissiveAdjustments}
  }
  MaterialExpressionMaterialFunctionCall_7 {
   class=MaterialExpressionMaterialFunctionCall
   inputs [
    UorV {from=MaterialExpressionConvert_1}
    Invert_EdgeControl {from=MaterialExpressionVectorParameter_10, output=RGB}
    NoInvert_EdgeControl {from=MaterialExpressionVectorParameter_2, output=RGB}
   ]
   properties {materialFunction=/Game/Effects/Niagara/MasterMaterials/Functions/Common/FXCommon_EdgeMask_Ex}
  }
  MaterialExpressionMultiply_0 {
   class=MaterialExpressionMultiply
   inputs [
    A {from=MaterialExpressionMultiply_13}
    B {from=MaterialExpressionDynamicParameter_0, output=R}
   ]
  }
  MaterialExpressionMultiply_1 {
   class=MaterialExpressionMultiply
   inputs [
    A {from=MaterialExpressionMultiply_0}
    B {from=MaterialExpressionVertexNormalWS_0}
   ]
  }
  MaterialExpressionMultiply_10 {
   class=MaterialExpressionMultiply
   inputs [
    A {from=MaterialExpressionAdd_2}
    B {from=MaterialExpressionScalarParameter_7}
   ]
  }
  MaterialExpressionMultiply_11 {
   class=MaterialExpressionMultiply
   inputs [
    A {from=MaterialExpressionDynamicParameter_1, output=A}
    B {from=""}
   ]
   properties {constB=0.5732}
  }
  MaterialExpressionMultiply_13 {
   class=MaterialExpressionMultiply
   inputs [
    A {from=MaterialExpressionPower_1}
    B {from=MaterialExpressionMaterialFunctionCall_2, output=Result}
   ]
  }
  MaterialExpressionMultiply_2 {
   class=MaterialExpressionMultiply
   inputs [
    A {from=MaterialExpressionMultiply_3}
    B {from=MaterialExpressionParticleColor_0, output=A}
   ]
  }
  MaterialExpressionMultiply_3 {
   class=MaterialExpressionMultiply
   inputs [
    A {from=MaterialExpressionDepthFade_0}
    B {from=MaterialExpressionMultiply_8}
   ]
  }
  MaterialExpressionMultiply_4 {
   class=MaterialExpressionMultiply
   inputs [
    A {from=MaterialExpressionFresnel_0}
    B {from=MaterialExpressionMultiply_5}
   ]
  }
  MaterialExpressionMultiply_5 {
   class=MaterialExpressionMultiply
   inputs [
    A {from=MaterialExpressionSkyLightEnvMapSample_0}
    B {from=MaterialExpressionVectorParameter_8, output=RGB}
   ]
  }
  MaterialExpressionMultiply_6 {
   class=MaterialExpressionMultiply
   inputs [
    A {from=MaterialExpressionScalarParameter_10}
    B {from=MaterialExpressionAdd_3}
   ]
  }
  MaterialExpressionMultiply_7 {
   class=MaterialExpressionMultiply
   inputs [
    A {from=MaterialExpressionAdd_1}
    B {from=MaterialExpressionScalarParameter_19}
   ]
  }
  MaterialExpressionMultiply_8 {
   class=MaterialExpressionMultiply
   inputs [
    A {from=MaterialExpressionSaturate_3}
    B {from=MaterialExpressionScalarParameter_21}
   ]
  }
  MaterialExpressionMultiply_9 {
   class=MaterialExpressionMultiply
   inputs [
    A {from=MaterialExpressionMaterialFunctionCall_7, output=Mask}
    B {from=MaterialExpressionComponentMask_1}
   ]
  }
  MaterialExpressionParticleColor_0 {class=MaterialExpressionParticleColor, inputs [], properties {}}
  MaterialExpressionPower_1 {
   class=MaterialExpressionPower
   inputs [
    Base {from=MaterialExpressionMultiply_9}
    Exp {from=MaterialExpressionScalarParameter_1}
   ]
  }
  MaterialExpressionPower_5 {
   class=MaterialExpressionPower
   inputs [
    Base {from=MaterialExpressionSaturate_4}
    Exp {from=MaterialExpressionScalarParameter_20}
   ]
  }
  MaterialExpressionPower_6 {
   class=MaterialExpressionPower
   inputs [
    Base {from=MaterialExpressionSaturate_2}
    Exp {from=MaterialExpressionScalarParameter_8}
   ]
  }
  MaterialExpressionReflectionVectorWS_0 {class=MaterialExpressionReflectionVectorWS, inputs [CustomWorldNormal {from=MaterialExpressionReroute_0, output=""}], properties {}}
  MaterialExpressionReroute_0 {class=MaterialExpressionReroute, inputs [None {from=MaterialExpressionTransform_0}]}
  MaterialExpressionReroute_1 {class=MaterialExpressionReroute, inputs [None {from=MaterialExpressionMultiply_2}]}
  MaterialExpressionSaturate_2 {class=MaterialExpressionSaturate, inputs [None {from=MaterialExpressionMultiply_10}]}
  MaterialExpressionSaturate_3 {class=MaterialExpressionSaturate, inputs [None {from=MaterialExpressionPower_6}]}
  MaterialExpressionSaturate_4 {class=MaterialExpressionSaturate, inputs [None {from=MaterialExpressionMultiply_7}]}
  MaterialExpressionSaturate_5 {class=MaterialExpressionSaturate, inputs [None {from=MaterialExpressionPower_5}]}
  MaterialExpressionScalarParameter_1 {class=MaterialExpressionScalarParameter, properties {defaultValue=2, parameterName=WPOFalloff}}
  MaterialExpressionScalarParameter_10 {class=MaterialExpressionScalarParameter, properties {defaultValue=-0.4, parameterName=Speed}}
  MaterialExpressionScalarParameter_11 {class=MaterialExpressionScalarParameter, properties {defaultValue=1.3, parameterName=sR}}
  MaterialExpressionScalarParameter_12 {class=MaterialExpressionScalarParameter, properties {defaultValue=0.6, parameterName=TLL}}
  MaterialExpressionScalarParameter_16 {class=MaterialExpressionScalarParameter, properties {parameterName=R2}}
  MaterialExpressionScalarParameter_18 {class=MaterialExpressionScalarParameter, properties {defaultValue=-0.2432, group=ColorG, parameterName=ColorG_GrayAdjust_Add}}
  MaterialExpressionScalarParameter_19 {class=MaterialExpressionScalarParameter, properties {defaultValue=0.8256, group=ColorG, parameterName=ColorG_GrayAdjust_Mul}}
  MaterialExpressionScalarParameter_20 {class=MaterialExpressionScalarParameter, properties {defaultValue=2, group=ColorG, parameterName=ColorG_GrayAdjust_Power}}
  MaterialExpressionScalarParameter_21 {class=MaterialExpressionScalarParameter, properties {defaultValue=0.5, parameterName=TotalOpacityScale}}
  MaterialExpressionScalarParameter_22 {class=MaterialExpressionScalarParameter, properties {defaultValue=1.5, parameterName=R3}}
  MaterialExpressionScalarParameter_3 {class=MaterialExpressionScalarParameter, properties {defaultValue=200, parameterName=DF}}
  MaterialExpressionScalarParameter_4 {class=MaterialExpressionScalarParameter, properties {defaultValue=20, parameterName=FresE}}
  MaterialExpressionScalarParameter_5 {class=MaterialExpressionScalarParameter, properties {defaultValue=0.1, parameterName=FresBaseReflectFractionIn}}
  MaterialExpressionScalarParameter_6 {class=MaterialExpressionScalarParameter, properties {group=TotalOpacity, parameterName=TotalOpacity_GrayAdjust_Add}}
  MaterialExpressionScalarParameter_7 {class=MaterialExpressionScalarParameter, properties {defaultValue=1.4385, group=TotalOpacity, parameterName=TotalOpacity_GrayAdjust_Mul}}
  MaterialExpressionScalarParameter_8 {class=MaterialExpressionScalarParameter, properties {defaultValue=1, group=TotalOpacity, parameterName=TotalOpacity_GrayAdjust_Power}}
  MaterialExpressionScalarParameter_9 {class=MaterialExpressionScalarParameter, properties {defaultValue=0.1792, parameterName=CV}}
  MaterialExpressionSkyLightEnvMapSample_0 {
   class=MaterialExpressionSkyLightEnvMapSample
   inputs [
    Direction {from=MaterialExpressionReflectionVectorWS_0, output=""}
    Roughness {from=MaterialExpressionScalarParameter_16, output=""}
   ]
   properties {}
  }
  MaterialExpressionTextureCoordinate_0 {class=MaterialExpressionTextureCoordinate}
  MaterialExpressionTextureCoordinate_1 {class=MaterialExpressionTextureCoordinate}
  MaterialExpressionTextureObjectParameter_0 {
   class=MaterialExpressionTextureObjectParameter
   inputs [
    UVs {from="", output=""}
    Apply View MipBias {from="", output=""}
   ]
   properties {
    group=AdditionNoise
    parameterName=AdditionNoise_Texture
    samplerType=SAMPLERTYPE_Masks
    sortPriority=0
    texture=/FortWater/Textures/T_OceanFoam_New_Mask
   }
  }
  MaterialExpressionTime_0 {class=MaterialExpressionTime, properties {bOverride_Period=true, period=6000}}
  MaterialExpressionTransform_0 {class=MaterialExpressionTransform, inputs [None {from=MaterialExpressionComponentMask_0}]}
  MaterialExpressionVectorParameter_0 {class=MaterialExpressionVectorParameter, properties {parameterName=UVS, defaultValue=<rgba 1 2.5 1 1>}}
  MaterialExpressionVectorParameter_1 {class=MaterialExpressionVectorParameter, properties {parameterName=AColor, defaultValue=<rgba 0 0.1322 0.2 1>}}
  MaterialExpressionVectorParameter_10 {
   class=MaterialExpressionVectorParameter
   properties {
    group=WPOEdgeMask
    parameterName=WPOEdgeMask_EdgeMask_InvertControl
    channelNames {a="", b=Power, g=Multiply, r=Offset}
    defaultValue=<rgba 0 2 1 1>
   }
  }
  MaterialExpressionVectorParameter_2 {
   class=MaterialExpressionVectorParameter
   properties {
    group=WPOEdgeMask
    parameterName=WPOEdgeMask_EdgeMask_Control
    channelNames {a="", b=Power, g=Multiply, r=Offset}
    defaultValue=<rgba 0 2 1 1>
   }
  }
  MaterialExpressionVectorParameter_3 {
   class=MaterialExpressionVectorParameter
   properties {
    group=AdditionNoise
    parameterName=AdditionNoise_UVControl
    channelNames {a=VPan, b=UPan, g=VScale, r=UScale}
    defaultValue=<rgba 0.15 0.15 0.3 0>
   }
  }
  MaterialExpressionVectorParameter_6 {class=MaterialExpressionVectorParameter, properties {parameterName=MSR, defaultValue=<rgba 0 0.2 0.3 1>}}
  MaterialExpressionVectorParameter_8 {class=MaterialExpressionVectorParameter, properties {parameterName=Ref, defaultValue=<rgba 0.5515 0.35 0.7 1>}}
  MaterialExpressionVectorParameter_9 {class=MaterialExpressionVectorParameter, properties {parameterName=BColor, defaultValue=<rgba 0 2 1.4261 1>}}
  MaterialExpressionVertexNormalWS_0 {class=MaterialExpressionVertexNormalWS, inputs [], properties {}}
 }
 properties {
  bAutomaticallySetUsageInEditor=false
  bUsedWithNiagaraMeshParticles=true
  blendMode=BLEND_Translucent
  refractionMethod=RM_IndexOfRefraction
  translucencyLightingMode=TLM_VolumetricDirectional
  usedShadingModels=Default Lit
 }
}
